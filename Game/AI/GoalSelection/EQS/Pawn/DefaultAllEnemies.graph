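# EQS query for AI goal selection that gathers enemy pawns, labeled as the default all-enemies goal for a null encounter.
# A single Fort enemies generator is not restricted to perceived enemies, except for sleep-capable AI, and skips downed (DBNO) pawns.
# Ability-range and propagated enemies are not added.
# It has no tests, so the query itself applies no scoring or further filtering.
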EnvQuery /Game/AI/GoalSelection/EQS/Pawn/DefaultAllEnemies
options [{
  generator {
   name=FortQueryGenerator_Enemies_12
   class=FortQueryGenerator_Enemies
   properties {
    bAddEnemiesFromAbilityRange=false
    bAutoSortTests=true
    bCanRunAsync=false
    bIgnoreDBNOPawns=true
    bIgnoreSleepingAIs=false
    bIncludeMarkedSleepingAIs=false
    bIncludePropagatedEnemies=false
    bPerceivedEnemiesOnly=false
    bSleepCapableAIUsePerceivedEnemiesOnly=true
    envQueryResultNormalizationOption=Default
    maxTimeSincePerceived=15
    optionName=All Enemies Default Goal For NULL Encounter
    abilityTags {gameplayTags []}
    distanceTags {gameplayTags []}
   }
  }
  tests []
 }]
